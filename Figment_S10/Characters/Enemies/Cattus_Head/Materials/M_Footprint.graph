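# Lit footprint material for the Cattus Head enemy, shaped by its mound mask and normal map.
# A Winter static switch, on by default, picks ice blended with landscape snow, or height-blended puddle, mud and gravel meeting grass or world-aligned rubble.
# A baked normal adds mound relief, with faint panning puddle ripples when non-winter, before grass color matching and a normal space transform.
# Several dirt and rubble blend chains feed nothing.
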Material /Figment_S10/Characters/Enemies/Cattus_Head/Materials/M_Footprint
graph {
 properties {
  bTangentSpaceNormal=false
  bUseMaterialAttributes=true
  pixelDepthOffsetMode=PDOM_Legacy
  refractionMethod=RM_IndexOfRefraction
  usedShadingModels=Default Lit
 }
 nodes [
  n1 = ScalarParameter() {defaultValue=900, group=Gravel, parameterName=VPDirtTextureTiling}
  n2 = MaterialFunctionCall(
   TextureObject=TextureObject() {texture=/Game/Packages/DS_Fortnight/WhiteRockCustomBR_D}
   TextureSize=@n1
   Export Float 4=StaticBool() {value=true}
  ) {materialFunction=&WorldAlignedTexture}
  n3 = Saturate(None=LinearInterpolate(A=Constant() {r=-1}, B=Constant() {r=1}, Alpha=ComponentMask(None=@n2.Z Texture) {a=true}))
  LinearInterpolate(A=Constant() {r=0.05}, B=Constant() {r=0.5}, Alpha=@n3) {constB=0.4}
  LinearInterpolate(A=Constant() {r=0.99}, B=Constant() {r=0.65}, Alpha=@n3)
  n4 = VertexColor()
  n5 = Multiply(A=@n4.B, B=OneMinus(None=@n4.A))
  n6 = ScalarParameter() {defaultValue=2, group=Dirt, parameterName=DirtBlendX-B}
  n7 = TextureObjectParameter() {group=Dirt, parameterName=Dirt, samplerSource=SSM_Wrap_WorldGroupSettings, texture=&WhiteRockCustomBR_D}
  n8 = TextureObjectParameter() {
   group=Dirt
   parameterName=Dirt_SpecMask
   samplerSource=SSM_Wrap_WorldGroupSettings
   texture=/FigmentEnvironment/FN_6_20/Characters/Fortress_Character/Textures/AllBlack
  }
  n9 = TextureObjectParameter() {
   group=Dirt
   parameterName=DirtTextureNormal
   samplerSource=SSM_Wrap_WorldGroupSettings
   samplerType=SAMPLERTYPE_Normal
   texture=/Game/Packages/DS_Fortnight/WhiteRockCustom_N
  }
  n10 = MaterialFunctionCall!(
   Diffuse Texture Object=@n7
   Spec Texture Object=@n8
   TextureSize=ScalarParameter() {defaultValue=900, group=Dirt, parameterName=DirtTextureTilingFrequency}
   --------------------------
   Use seperate settings for Z projection
   Use seperate texture for Z projection
   Texture Size Z projection
   Z Plane Diffuse Texture Object
   Z Plane Spec Texture Object
   "--------------------------      "
   ProjectionTransitionContrast
   Use Texture Alpha Heightmap to Refine the Transistion
   HeightLerpTransitionContrast
   PickTheDiffuseHeightChannel
   TheDiffuseHeightChannel
   "--------------------------       "
   NormalTextureObject=@n9
   World Space Normals
   "--------------------------  "
   Export Float 4 Diffuse=StaticBool() {value=true}
   Export Float 4 Spec
   "--------------------------     "
   WorldPosition
   World Space Normal
  ) {materialFunction=/Engine/Functions/Engine_MaterialFunctions02/Texturing/WorldAlignedTextures_Complex}
  n11 = MaterialFunctionCall!(
   Mat attributes=MaterialFunctionCall!(In=@n10.Z Mat Attributes - See Tooltip) {materialFunction=/Game/Packages/Fortress_SharedMaterials/Base_Material_Functions/SwizzleWorldAlignedSpecOutputForFortnite}.Result
   Use Spec=StaticBoolParameter() {group=Dirt, parameterName=Use Dirt Spec}
   Use Normals=StaticBoolParameter() {defaultValue=true, group=Dirt, parameterName=Use Dirt Normals}
  ) {materialFunction=/Game/Packages/Fortnite_UniqueMaterials_1/MaterialFunctions/Fortnite_Terrain_invertOpacityNormalBool}
  n12 = GetMaterialAttributes(None=@n11.Result) {attributeGetTypes [#1, #2, #3]}
  n13 = OneMinus(None=@n12.Opacity)
  n14 = Clamp(None=Power(Base=Multiply(A=@n6, B=@n13), Exp=ScalarParameter() {defaultValue=5, group=Dirt, parameterName=DirtBlendPower-B}))
  n15 = Multiply(A=ScalarParameter() {defaultValue=2, group=Dirt, parameterName=DirtBlendX-A}, B=OneMinus(None=@n12.Opacity))
  n16 = ScalarParameter() {defaultValue=1, group=Dirt, parameterName=DirtBlendPower-A}
  n17 = MaterialFunctionCall(Transition Phase=VertexColor().B, Height Texture=Clamp(None=Power(Base=@n15, Exp=@n16)), Contrast=Constant() {r=1}) {desc=mud + grass blended, materialFunction=&HeightLerp}
  n18 = Constant() {r=1}
  n19 = Multiply(A=@n12.Base Color, B=VectorParameter() {group=Dirt, parameterName=DirtTint, defaultValue=<rgba 1 1 1 0>}.RGB)
  n20 = CollectionParameter() {collection=/Game/Athena/Environments/Landscape/MPC/MPC_Landscape, parameterName=Winter_Gravel_Tint}
  n21 = StaticSwitchParameter(
   True=Multiply(
    A=MaterialFunctionCall() {materialFunction=/Game/Athena/Environments/Landscape/MaterialFunctions/Standard/MF_Athena_RoadGravel}.Gravel_NoColor
    B=ComponentMask(None=@n20) {b=true, g=true, r=true}
   )
   False=@n19
  ) {parameterName=ForceSnow}
  n22 = Constant3Vector() {constant=<rgba 0 0 1 0>}
  n23 = ScalarParameter() {defaultValue=768, group=Rubble, parameterName=RubbleScale}
  n24 = MaterialFunctionCall(Float3=MaterialFunctionCall(
    TextureObject=TextureObject() {samplerType=SAMPLERTYPE_Masks, texture=/Game/Environments/Sets/Rubble/Props/Textures/T_rubble_pile_S}
    TextureSize=@n23
   ) {materialFunction=/Game/Weapons/WeaponSkins/Wraps/MaterialFunctions/WorldAlignedTexture}.Z Texture) {materialFunction=&BreakOutFloat3Components}
  n25 = SetMaterialAttributes!(
   MaterialAttributes
   Base Color=MaterialFunctionCall(TextureObject=TextureObject() {texture=/Figment_S10/Environments/Sets/Rubble/Props/Textures/T_rubble_pile_02_D}, TextureSize=@n23) {materialFunction=&WorldAlignedTexture}.Z Texture
   Specular=@n24.R
   Roughness=@n24.R
   Normal=MaterialFunctionCall(
    TextureObject=TextureObject() {samplerType=SAMPLERTYPE_Normal, texture=/Game/Environments/Sets/Rubble/Props/Textures/T_rubble_pile_N}
    TextureSize=@n23
   ) {materialFunction=&WorldAlignedNormal}.Z Texture
  ) {attributeSetTypes [
    #1
    #4
    #5
    #3
   ]}
  MaterialFunctionCall!(
   Base Material=@n25
   Top Material=SetMaterialAttributes!(
    MaterialAttributes=@n12.MaterialAttributes
    Base Color=StaticSwitchParameter(True=@n19, False=@n21) {parameterName=ForceRegularGravel}
    Specular=ScalarParameter() {defaultValue=0.5, parameterName=DirtSpecValue}
    Normal=StaticSwitchParameter(True=@n22, False=@n12.Normal) {parameterName=ForceSnow}
   ) {attributeSetTypes [#1, #4, #3]}
   Alpha=LinearInterpolate(
    A=MaterialFunctionCall(Transition Phase=@n5, Height Texture=@n14) {desc=mud + grass blended, materialFunction=&HeightLerp}.Lerp Alpha No Contrast
    B=MaterialFunctionCall(In=@n17.Alpha, Contrast=@n18) {materialFunction=&CheapContrast}.Result
    Alpha=VertexColor().B
   )
  ) {materialFunction=/Engine/Functions/MaterialLayerFunctions/MatLayerBlend_Standard}
  n26 = TextureObjectParameter() {group=Dirt, parameterName=Dirt, samplerSource=SSM_Wrap_WorldGroupSettings, texture=&WhiteRockCustomBR_D}
  n27 = TextureObjectParameter() {
   group=Dirt
   parameterName=Dirt_SpecMask
   samplerSource=SSM_Wrap_WorldGroupSettings
   texture=&AllBlack
  }
  n28 = TextureObjectParameter() {
   group=Dirt
   parameterName=DirtTextureNormal
   samplerSource=SSM_Wrap_WorldGroupSettings
   samplerType=SAMPLERTYPE_Normal
   texture=&WhiteRockCustom_N
  }
  n29 = MaterialFunctionCall(
   Diffuse Texture Object=@n26
   Spec Texture Object=@n27
   TextureSize=ScalarParameter() {defaultValue=900, group=Dirt, parameterName=DirtTextureTilingFrequency}
   NormalTextureObject=@n28
   Export Float 4 Diffuse=StaticBool() {value=true}
  ) {materialFunction=&WorldAlignedTextures_Complex}
  MaterialFunctionCall(
   Mat attributes=MaterialFunctionCall(In=@n29.Z Mat Attributes - See Tooltip) {materialFunction=&SwizzleWorldAlignedSpecOutputForFortnite}.Result
   Use Spec=StaticBoolParameter() {group=Dirt, parameterName=Use Dirt Spec}
   Use Normals=StaticBoolParameter() {defaultValue=true, group=Dirt, parameterName=Use Dirt Normals}
  ) {materialFunction=&Fortnite_Terrain_invertOpacityNormalBool}
  Multiply(
   A=VectorParameter() {group=Dirt, parameterName=LerpDirtColor, defaultValue=<rgba 0.625 0.5868 0.5375 1>}.RGB
   B=Power(Base=@n19, Exp=ScalarParameter() {defaultValue=1, group=Dirt, parameterName=DirtBlendPower-A_1})
  )
  n30 = VectorParameter() {parameterName=IceColor, defaultValue=<rgba 0.0762 0.1743 0.2812 1>}
  n31 = ScalarParameter() {defaultValue=0.2, parameterName=IceRaughness}
  n32 = GetMaterialAttributes(None=MaterialFunctionCall!(UseLandscapeCoordsOnSwitch) {materialFunction=/Game/Athena/Environments/Landscape/MaterialFunctions/Snow/MF_Athena_Snow_01}.Material) {attributeGetTypes [#1, #5, #4]}
  n33 = TextureSampleParameter2D() {parameterName=Mask, samplerType=SAMPLERTYPE_Masks, texture=/Figment_S10/Characters/Enemies/Cattus_Head/Textures/S_Cattus_Mound_D}
  n34 = LinearInterpolate(Alpha=@n33.R) {constA=0.5}
  n35 = Multiply(A=@n32.Base Color, B=@n34)
  n36 = Multiply(A=@n32.Roughness, B=@n34)
  n37 = Multiply(A=@n32.Specular, B=@n33.R)
  n38 = SetMaterialAttributes!(
   MaterialAttributes
   Base Color=ComponentMask(None=@n2.Z Texture) {b=true, g=true, r=true}
   Specular=ScalarParameter() {group=Gravel, parameterName=Mud S}
   Roughness=ScalarParameter() {group=Gravel, parameterName=Mud R}
   Normal=MaterialFunctionCall(TextureObject=TextureObject() {samplerType=SAMPLERTYPE_Normal, texture=&WhiteRockCustom_N}, TextureSize=@n1) {materialFunction=&WorldAlignedNormal}.Z Texture
  ) {attributeSetTypes [
    #1
    #4
    #5
    #3
   ]}
  n39 = GetMaterialAttributes(None=@n38) {attributeGetTypes [#1]}
  n40 = VectorParameter() {parameterName=VP_DirtTint, defaultValue=<rgba 0.9375 0.9375 0.9375 0>}
  n41 = MaterialFunctionCall(
   Base Material=SetMaterialAttributes!(MaterialAttributes=@n39.MaterialAttributes, Base Color=Multiply(A=@n39.Base Color, B=@n40.RGB)) {attributeSetTypes [#1]}
   Top Material=@n25
   Alpha=MaterialFunctionCall(Transition Phase=VertexColor().A, Height Texture=ComponentMask(None=@n29.Z Diffuse Texture) {a=true}) {materialFunction=&HeightLerp}.Alpha
  ) {materialFunction=&MatLayerBlend_Standard}
  n42 = MaterialFunctionCall() {materialFunction=/Game/Athena/Environments/Landscape/MaterialFunctions/Standard/MF_Athena_Grass_01}
  n43 = ScalarParameter() {defaultValue=0.499, group=Gravel, parameterName=GravelMud/Grass Trans}
  n44 = Multiply(A=TextureCoordinate(), B=ScalarParameter() {defaultValue=7, group=Mud, parameterName=Mud Scale})
  n45 = TextureSample(UVs=@n44) {texture=/Game/Athena/Environments/Island/Textures/Dirt_D}
  n46 = Subtract(A=@n33.R, B=Multiply(A=@n45.A, B=ScalarParameter() {defaultValue=0.35, group=Gravel, parameterName=Mud Height}))
  n47 = ScalarParameter() {defaultValue=2, group=Gravel, parameterName=GravelMud/Grass Contrast}
  n48 = Multiply(A=Multiply(A=@n45.RGB) {constB=0.5}, B=OneMinus(None=Fresnel() {exponent=3}))
  n49 = ScalarParameter() {defaultValue=0.1, group=Puddle, parameterName=Puddle Spec}
  n50 = ScalarParameter() {defaultValue=0.0348, group=Puddle, parameterName=Puddle Rough}
  n51 = TextureSampleParameter2D() {parameterName=Mask, samplerType=SAMPLERTYPE_Masks, texture=&S_Cattus_Mound_D}
  n52 = MakeMaterialAttributes!(
   BaseColor=@n45.RGB
   Metallic
   Specular=LinearInterpolate(
    A=ScalarParameter() {defaultValue=0.2, group=Mud, parameterName=SpecMud}
    B=ScalarParameter() {defaultValue=0.5, group=Mud, parameterName=Spec Mud Top}
    Alpha=@n51.G
   )
   Roughness=LinearInterpolate(
    A=ScalarParameter() {defaultValue=0.3, group=Mud, parameterName=Mud R Bott}
    B=ScalarParameter() {defaultValue=0.85, group=Mud, parameterName=Mud R Top}
    Alpha=@n51.R
   )
   Anisotropy
   EmissiveColor
   Opacity
   OpacityMask
   Normal=MaterialFunctionCall(
    Normal=TextureSample(UVs=@n44) {samplerType=SAMPLERTYPE_Normal, texture=/Game/Packages/DS_Fortnight/ExperimentalTextures/Dirt_N}.RGB
    Flatness=ScalarParameter() {defaultValue=0.1614, group=Mud, parameterName=Mud Flatten}
   ) {materialFunction=&FlattenNormal}.Result
   Tangent
   WorldPositionOffset
   SubsurfaceColor
   ClearCoat
   ClearCoatRoughness
   AmbientOcclusion
   Refraction
   CustomizedUVs_0
   CustomizedUVs_1
   CustomizedUVs_2
   CustomizedUVs_3
   CustomizedUVs_4
   CustomizedUVs_5
   CustomizedUVs_6
   CustomizedUVs_7
   PixelDepthOffset
   ShadingModel
   Displacement
  )
  n53 = Multiply(A=TextureCoordinate(), B=ScalarParameter() {defaultValue=4.8933, group=Gravel, parameterName=MudTextureTiling})
  n54 = TextureSample(UVs=@n53) {texture=/Game/Packages/DS_Fortnight/WhiteRockCustom_D}
  n55 = Saturate(None=LinearInterpolate(A=Constant() {r=-1}, B=Constant() {r=1}, Alpha=@n54.A))
  n56 = SetMaterialAttributes!(
   MaterialAttributes
   Base Color=@n54.RGB
   Specular=LinearInterpolate(A=Constant() {r=0.05}, B=Constant() {r=0.5}, Alpha=@n55) {constB=0.4}
   Roughness=LinearInterpolate(A=Constant() {r=0.99}, B=Constant() {r=0.65}, Alpha=@n55)
   Normal=TextureSample(UVs=@n53) {samplerType=SAMPLERTYPE_Normal, texture=&WhiteRockCustom_N}.RGB
  ) {attributeSetTypes [
    #1
    #4
    #5
    #3
   ]}
  n57 = TextureSampleParameter2D() {parameterName=Mask, samplerType=SAMPLERTYPE_Masks, texture=&S_Cattus_Mound_D}
  n58 = MaterialFunctionCall(
   Transition Phase=ScalarParameter() {defaultValue=0.419, group=Gravel, parameterName=Gravel/Mud Transition}
   Height Texture=Add(A=@n57.R, B=Multiply(A=@n57.R, B=@n54.A))
   Contrast=ScalarParameter() {defaultValue=1, group=Gravel, parameterName=GravelMud Contrast}
  ) {materialFunction=&HeightLerp}
  n59 = MaterialFunctionCall(Transition Phase=Constant() {r=0.65}, Height Texture=@n51.R, Contrast=Constant() {r=2}) {materialFunction=&HeightLerp}
  n60 = GetMaterialAttributes(None=MaterialFunctionCall(
    Base Material=MakeMaterialAttributes(BaseColor=@n48, Specular=@n49, Roughness=@n50)
    Top Material=MaterialFunctionCall(Base Material=@n52, Top Material=@n56, Alpha=@n58.Alpha) {materialFunction=&MatLayerBlend_Standard}.Blended Material
    Alpha=@n59.Alpha
   ) {materialFunction=&MatLayerBlend_Standard}.Blended Material) {attributeGetTypes [#1]}
  n61 = TextureSampleParameter2D() {parameterName=Mask, samplerType=SAMPLERTYPE_Masks, texture=&S_Cattus_Mound_D}
  n62 = Multiply(
   A=@n60.Base Color
   B=LinearInterpolate(
    A=VectorParameter() {group=Mud, parameterName=Mud Tint Low, defaultValue=<rgba 0.1146 0.098 0.0525 1>}.RGB
    B=VectorParameter() {group=Mud, parameterName=Mut Tint Top, defaultValue=<rgba 0.6979 0.583 0.4358 1>}.RGB
    Alpha=@n61.R
   )
  )
  n63 = StaticSwitchParameter(
   True=MaterialFunctionCall(
    Base Material=MakeMaterialAttributes(BaseColor=@n30.RGB, Roughness=@n31)
    Top Material=SetMaterialAttributes!(MaterialAttributes=@n32.MaterialAttributes, Base Color=@n35, Roughness=@n36, Specular=@n37) {attributeSetTypes [#1, #5, #4]}
    Alpha=@n33.R
   ) {materialFunction=&MatLayerBlend_Standard}.Blended Material
   False=MaterialFunctionCall(
    Base Material=SetMaterialAttributes!(MaterialAttributes=@n60.MaterialAttributes, Base Color=@n62) {attributeSetTypes [#1]}
    Top Material=StaticSwitchParameter(True=@n41.Blended Material, False=@n42.Result) {group=Biome, parameterName=Rubble}
    Alpha=MaterialFunctionCall(Transition Phase=@n43, Height Texture=@n46, Contrast=@n47) {materialFunction=&HeightLerp}.Alpha
   ) {materialFunction=&MatLayerBlend_Standard}.Blended Material
  ) {defaultValue=true, group=Biome, parameterName=Winter}
  n64 = SetMaterialAttributes(Base Color=GetMaterialAttributes(None=@n63) {attributeGetTypes [#1]}.Base Color) {attributeSetTypes [#1]}
  n65 = TextureSampleParameter2D() {parameterName=Normal, samplerType=SAMPLERTYPE_Normal, texture=/Figment_S10/Characters/Enemies/Cattus_Head/Textures/S_Cattus_Mound_N}
  n66 = Divide(A=ComponentMask(None=WorldPosition().XYZ) {g=true, r=true}, B=Constant() {r=250})
  n67 = TextureSample(UVs=Panner(Coordinate=@n66) {speedX=0.1, speedY=0.1}) {samplerSource=SSM_Wrap_WorldGroupSettings, samplerType=SAMPLERTYPE_Normal, texture=/Figment_S10/Environments/Foliage/Textures/T_SwirlNoise_N}
  n68 = TextureSample(UVs=Panner(Coordinate=@n66) {speedX=-0.1, speedY=-0.1}) {samplerSource=SSM_Wrap_WorldGroupSettings, samplerType=SAMPLERTYPE_Normal, texture=&T_SwirlNoise_N}
  n69 = Add(A=@n67.RGB, B=Multiply(A=@n68.RGB, B=Constant3Vector() {constant=<rgba 1 1 0 0>}))
  n70 = Constant() {r=0.99}
  n71 = ShadingPathSwitch(
   Default=LinearInterpolate(
    A=MaterialFunctionCall(Normal=@n69, Flatness=@n70) {materialFunction=&FlattenNormal}.Result
    B=@n65.RGB
    Alpha=@n59.Alpha
   )
   Forward=LinearInterpolate(A=Constant3Vector() {constant=<rgba 0 0 1 0>}, B=@n65.RGB, Alpha=@n59.Alpha)
   Mobile=@n65.RGB
  )
  n72 = GetMaterialAttributes(None=MaterialFunctionCall!(
    Material=ShadingPathSwitch(Default=@n63, Deferred=@n63, Forward=@n63, Mobile=@n64)
    Normal=StaticSwitchParameter(True=@n65.RGB, False=@n71) {defaultValue=true, group=Biome, parameterName=Winter}
   ) {materialFunction=/Engine/Functions/MaterialLayerFunctions/MatLayerBlend_BakedNormal}.Blended Material) {attributeGetTypes [#1]}
  n73 = GetMaterialAttributes(None=SetMaterialAttributes(
    MaterialAttributes=@n72.MaterialAttributes
    Base Color=MaterialFunctionCall!(BaseColor=@n72.Base Color) {materialFunction=/Game/Athena/Environments/Landscape/MaterialFunctions/Standard/MF_Athena_GrassMatch}.Result
   ) {attributeSetTypes [#1]}) {attributeGetTypes [#3]}
  SetMaterialAttributes!(MaterialAttributes=@n73.MaterialAttributes, Normal=Transform(None=@n73.Normal)) {attributeSetTypes [#3]}
  n74 = ComponentMask(None=@n10.Z Diffuse Texture) {r=true}
  n75 = ScalarParameter() {defaultValue=0.2, group=Dirt, parameterName=DirtDiffuseSpecMod}
  StaticSwitchParameter(
   True=SetMaterialAttributes!(
    MaterialAttributes=@n11.Result
    Roughness=ScalarParameter() {defaultValue=0.5, group=Dirt, parameterName=DirtRoughnessMod}
    Specular=Multiply(A=@n74, B=@n75)
   ) {attributeSetTypes [#5, #4]}
   False=@n11.Result
  ) {defaultValue=true, group=Dirt, parameterName=UseDirtDiffuseAsSpec}
 ]
}
